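# Translucent unlit Niagara sprite material for a heat distortion effect, using index-of-refraction refraction.
# A panning Perlin normal texture supplies moving noise, and a radial falloff from the UV center, shaped by power and smoothstep, masks it.
# Mask, particle alpha and the noise red channel drive a lerp from 1 toward 0, while a separate multiply masks the normal RGB.
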Material /VividRazorVaultGameplay/FX/M_VividRazor_HeatDistortion
graph {
 properties {
  bAutomaticallySetUsageInEditor=false
  bUsedWithNiagaraSprites=true
  blendMode=BLEND_Translucent
  refractionMethod=RM_IndexOfRefraction
  shadingModel=MSM_Unlit
  usedShadingModels=Unlit
 }
 nodes [
  n1 = TextureCoordinate()
  n2 = Length(None=Subtract(A=@n1) {constB=0.5}) {is=class MaterialExpressionLength}
  n3 = SmoothStep(Value=Power(Base=OneMinus(None=@n2)) {constExponent=3}) {constMin=0.2}
  n4 = ParticleColor()
  n5 = TextureSample(UVs=Panner(Coordinate=Multiply(A=@n1)) {speedX=0.45, speedY=0.6}) {samplerType=SAMPLERTYPE_Normal, texture=/Game/Effects/Fort_Effects/Textures/T_PerlinNormal}
  n6 = Max(A=@n5.R, B=@n5.R)
  LinearInterpolate(Alpha=Multiply(A=@n3, B=Multiply(A=@n4.A, B=@n6))) {constA=1, constB=0}
  Multiply(A=@n5.RGB, B=@n3)
 ]
}
