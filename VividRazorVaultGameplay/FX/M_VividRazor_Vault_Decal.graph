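# Translucent deferred decal material for the VividRazor Vault effect, drawing a glowing shockwave mark with an animated cellular pattern.
# Two world-position-mapped Voronoi samples pan in different directions and multiply together, tinted by the EmissiveColor parameter.
# A UV-distance radial falloff intensifies the center, and the shockwave texture masks the glow, which is scaled by decal color alpha and passed through MF_FX_EmissiveAdjustments.
# The same texture, multiplied by decal lifetime opacity, provides a fading mask.
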Material /VividRazorVaultGameplay/FX/M_VividRazor_Vault_Decal
graph {
 properties {bAutomaticallySetUsageInEditor=false, blendMode=BLEND_Translucent, materialDomain=MD_DeferredDecal, usedShadingModels=Default Lit}
 nodes [
  Abs()
  Constant() {r=1}
  n1 = WorldPosition()
  n2 = Time()
  n3 = Constant3Vector() {constant=<rgba 0.1 -0.21 0.1 0>}
  n4 = TextureSample(UVs=Add(A=Multiply(A=@n1.XYZ) {constB=0.003}, B=Multiply(A=@n2, B=@n3))) {samplerType=SAMPLERTYPE_LinearColor, texture=/Game/Effects/Explosion_Update/Textures/VT_Voronoi_Seamless}
  n5 = Time()
  n6 = Constant3Vector() {constant=<rgba -0.14 0.2 -0.2 0>}
  n7 = TextureSample(UVs=Add(A=Multiply(A=@n1.XYZ) {constB=0.004}, B=Multiply(A=@n5, B=@n6) {constB=0.2})) {samplerType=SAMPLERTYPE_LinearColor, texture=&VT_Voronoi_Seamless}
  n8 = Multiply(
   A=VectorParameter() {parameterName=EmissiveColor, defaultValue=<rgba 5 0.637 0 1>}.RGB
   B=Power(Base=Multiply(A=@n4.R, B=@n7.R)) {constExponent=1}
  )
  n9 = TextureCoordinate()
  n10 = OneMinus(None=Length(None=Subtract(A=@n9) {constB=0.5}) {is=class MaterialExpressionLength})
  n11 = SmoothStep(Value=@n10) {constMin=0.8}
  n12 = Add(A=Multiply(A=@n8, B=Power(Base=@n10) {constExponent=10}), B=Multiply(A=Multiply(A=@n11, B=@n8)) {constB=2})
  n13 = TextureSample(UVs=@n9) {samplerType=SAMPLERTYPE_LinearGrayscale, texture=&/T_VividRazor_Vault_Shockwave}
  n14 = Power(Base=@n13.R) {constExponent=4}
  MaterialFunctionCall(Emissive=Multiply(A=Multiply(A=@n12, B=@n14), B=DecalColor().A)) {materialFunction=&MF_FX_EmissiveAdjustments}
  Multiply(A=Multiply(A=@n13.R) {constB=2}, B=DecalLifetimeOpacity().Opacity)
  VectorParameter() {parameterName=DecalColor, defaultValue=<rgba 0.003 0 0 1>}
 ]
}
