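# Masked, lit material for Niagara mesh particles, rendering Vivid Razor vault paper notes tinted by particle color.
# As particle alpha falls, a tiling noise pattern, offset randomly per particle, drives a burn-style dissolve.
# Subtracting a sharpened threshold from the dissolve value isolates a band, colored HDR orange and passed through the shared FX emissive adjustment function.
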Material /VividRazorVaultGameplay/FX/M_VividRazor_Vault_Paper_VFX
graph {
 properties {bAutomaticallySetUsageInEditor=false, bUsedWithNiagaraMeshParticles=true, blendMode=BLEND_Masked, usedShadingModels=Default Lit}
 nodes [
  Constant()
  n1 = ParticleColor()
  n2 = OneMinus(None=@n1.A)
  n3 = Add(A=Multiply(A=TextureCoordinate()) {constB=0.4}, B=ParticleRandom())
  n4 = Saturate(None=Add(
    A=LinearInterpolate(Alpha=@n2) {constA=1, constB=-1}
    B=TextureSample(UVs=@n3) {texture=/Game/Effects/Fort_Effects/Textures/T_Tile_Pattern_01}.R
   ))
  n5 = Saturate(None=Multiply(A=Subtract(A=@n4) {constB=0.35}) {constB=4})
  MaterialFunctionCall(Emissive=Multiply(A=Subtract(A=@n4, B=@n5), B=Constant3Vector() {constant=<rgba 10 1.1234 0 1>})) {materialFunction=&MF_FX_EmissiveAdjustments}
  Multiply(A=TextureSample() {texture=&/T_Paper_Notes_D}.RGB, B=@n1.RGB)
  Time() {bOverride_Period=true, period=1}
 ]
}
